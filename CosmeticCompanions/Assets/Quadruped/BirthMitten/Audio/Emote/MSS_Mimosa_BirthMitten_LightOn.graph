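# Mono one-shot MetaSound source for the LightOn emote sound of the BirthMitten quadruped cosmetic companion.
# OnPlay triggers a single wave player with a fixed light emote wave, feeding the mono output and signalling OnFinished on completion.
# The affiliation, volume, pitch, looping, fade and intro/loop array inputs, plus the VoEnvelope output, are declared but unconnected.
# Companion foley attenuation, property sheet and emote concurrency settings govern playback.
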MetaSoundSource /CosmeticCompanions/Assets/Quadruped/BirthMitten/Audio/Emote/MSS_Mimosa_BirthMitten_LightOn
inputs [
 B {type=Enum:FortAudioAffiliation}
 DelayLoopByStart {type=Bool, access=Constructor}
 FadeInTime {type=Time, access=Constructor}
 FriendlyVolume {type=Float}
 Intro.Music.Sounds {type=WaveAsset:Array, access=Constructor}
 IntroVolumeMultiplier {type=Float}
 Loop.Music.Sounds {type=WaveAsset:Array, access=Constructor}
 Looping {type=Bool, access=Constructor}
 LoopVolumeMultiplier {type=Float}
 PitchTemplateMultiplier {type=Float}
 UE.Source.OnPlay {type=Trigger}
 UnfriendlyVolume {type=Float}
]
interfaces [UE.OutputFormat.Mono:1.0, UE.Source.OneShot:1.0, UE.Source:1.1]
outputs [
 "UE.OutputFormat.Mono.Audio:0" {type=Audio}
 UE.Source.OneShot.OnFinished {type=Trigger}
 VoEnvelope {type=Float}
]
pages [Default {
  input_defaults {
   B=null
   DelayLoopByStart="<MetaSoundBuilderResult.SUCCEEDED: 0>"
   FadeInTime=0
   FriendlyVolume=null
   IntroVolumeMultiplier=1
   LoopVolumeMultiplier=1
   Looping=null
   PitchTemplateMultiplier=1
   UE.Source.OnPlay=null
   UnfriendlyVolume=null
   Intro.Music.Sounds [&/IP_Emote_Companion_BirthMitten]
   Loop.Music.Sounds []
  }
  nodes [WavePlayer_0 = UE.Wave Player:Mono(Play=^UE.Source.OnPlay) {also {defaults {Wave Asset=&/IP_Emote_Companion_BirthMitten_Light}, version [1,1]}}]
  links [{dst="@UE.Source.OneShot.OnFinished", src=WavePlayer_0.On Finished}, {dst="@UE.OutputFormat.Mono.Audio:0", src=WavePlayer_0.Out Mono}]
 }]
properties {
 assetImportData=&self:AssetImportData
 attenuationSettings=/Game/Sounds/Attenuation/AS_Companion_Foley
 audioPropertiesSheet=/CosmeticCompanions/Assets/Audio_Shared/PropertySheets/APS_Cosmetic_Mimosa_Foley
 maxDistance=1200
 sampleRate=48000
 soundClassObject=&Fort_Creatures_Pets
 virtualizationMode=PlayWhenSilent
 concurrencySet [/Game/Sounds/Concurrency_Rules/Athena_EmoteFoley_Conc]
 soundSubmixSends [{}]
}
